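; BP_DebrisContainer is a debris actor with a suction effect and a proximity-driven post-process.
; On BeginPlay, the Should Spawn scalable float either keeps it, disabling the suction effect's custom depth after a short delay, or enables custom depth and destroys the actor.
; Any overlap entering PostFX_Capsule activates the PostProcess component, and any overlap ending deactivates it, without filtering by actor.
; Tick and actor overlap events are empty, and construction only calls the parent.
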(event EventBeginPlay
  (bind _ns_lw_debris_suction (Variables|Default|GetNS_LW_Debris_suction))
  (if (ScalableFloat|GetValueasBool (Variables|Default|GetShould%20Spawn))
    (Utilities|FlowControl|Delay 0.3)
    (Rendering|SetRenderCustomDepth _ns_lw_debris_suction false)
    (else
      (Rendering|SetRenderCustomDepth _ns_lw_debris_suction true)
      (Actor|DestroyActor))))

(event Collision|EventActorBeginOverlap
  (params (OtherActor :type Actor%20Object%20Reference)))

(event EventTick
  (params (DeltaSeconds :type Float%20%28single-precision%29)))

(event OnComponentBeginOverlap(PostFX_Capsule)
  (params (OverlappedComponent :type Primitive%20Component%20Object%20Reference) (OtherActor :type Actor%20Object%20Reference) (OtherComp :type Primitive%20Component%20Object%20Reference) (OtherBodyIndex :type Integer) (bFromSweep :type Boolean) (SweepResult :type Hit%20Result%20Structure%20%28by%20ref%29))
  (Components|Activation|Activate (Variables|Default|GetPostProcess) false))

(event OnComponentEndOverlap(PostFX_Capsule)
  (params (OverlappedComponent :type Primitive%20Component%20Object%20Reference) (OtherActor :type Actor%20Object%20Reference) (OtherComp :type Primitive%20Component%20Object%20Reference) (OtherBodyIndex :type Integer))
  (Components|Activation|Deactivate (Variables|Default|GetPostProcess)))

(fn ConstructionScript
  (|CallParentFunction))
