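; BP_DebrisContainer_LW is a debris container actor for WildEstate LowGrav that decides at BeginPlay whether it stays in the level.
; It reads the Should Spawn scalable float as a bool.
; If true, it waits briefly, then turns off custom depth rendering on its NS_LW_Debris_suction19 component; if false, it turns that on and destroys itself.
; Overlap and Tick events are empty, and the construction script only calls its parent.
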(event EventBeginPlay
  (bind _ns_lw_debris_suction19 (Variables|Default|GetNS_LW_Debris_suction19))
  (if (ScalableFloat|GetValueasBool (Variables|Default|GetShould%20Spawn))
    (Utilities|FlowControl|Delay 0.3)
    (Rendering|SetRenderCustomDepth _ns_lw_debris_suction19 false)
    (else
      (Rendering|SetRenderCustomDepth _ns_lw_debris_suction19 true)
      (Actor|DestroyActor))))

(event Collision|EventActorBeginOverlap
  (params (OtherActor :type Actor%20Object%20Reference)))

(event EventTick
  (params (DeltaSeconds :type Float%20%28single-precision%29)))

(fn ConstructionScript
  (|CallParentFunction))
